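; Essence item for Minor Spire sub-scenario B that activates a Niagara effect aimed using a nearby attacking minor spire.
; Shortly after BeginPlay, it runs a sphere overlap around itself filtered to BGA_Spire_Minor_Attacking actors.
; If any are found, it writes the first one's world location into the SystemDirection Niagara variable and activates the component.
; Overlap and Tick events are empty.
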(event EventBeginPlay
  (bind _niagara (Variables|Default|GetNiagara))
  (|CallParentFunction)
  (Utilities|FlowControl|Delay 0.4)
  (bind (_outactors _returnvalue) (Collision|SphereOverlapActors (Transformation|GetActorLocation) 2.5e+03 (Utilities|Array|MakeArray "ObjectTypeQuery3" "ObjectTypeQuery7" "ObjectTypeQuery2" "ObjectTypeQuery6" "ObjectTypeQuery1") "/SpireBoss/Gameplay/MinorSpire/BGA_Spire_Minor_Attacking.BGA_Spire_Minor_Attacking_C"))
  (if _returnvalue
    (Niagara|SetNiagaraVariable(Vector3) _niagara "SystemDirection" (Transformation|GetActorLocation (Utilities|Array|Get(acopy) _outactors)))
    (Components|Activation|Activate _niagara false)))

(event Collision|EventActorBeginOverlap
  (params (OtherActor :type Actor%20Object%20Reference)))

(event EventTick
  (params (DeltaSeconds :type Float%20%28single-precision%29)))

(fn ConstructionScript
  (|CallParentFunction))
