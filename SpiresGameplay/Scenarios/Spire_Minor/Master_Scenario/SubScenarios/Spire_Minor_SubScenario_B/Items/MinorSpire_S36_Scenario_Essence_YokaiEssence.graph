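# Translucent unlit material for the Yokai Essence scenario item, drawing a cyan-blue wispy flame with a bright core on a Sprite-function billboard.
# Wave-distorted, scaled UVs and panning noise warp sphere masks, combined with a scrolling wispy mask for shape and color.
# A polar-UV noise variant and top-view factor support overhead viewing, while an object-position pseudo-random value is declared for variation.
# Camera depth fade and time-of-day scaling attenuate the effect.
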Material /SpiresGameplay/Scenarios/Spire_Minor/Master_Scenario/SubScenarios/Spire_Minor_SubScenario_B/Items/MinorSpire_S36_Scenario_Essence_YokaiEssence
graph {
 properties {bAutomaticallySetUsageInEditor=false, blendMode=BLEND_Translucent, shadingModel=MSM_Unlit, usedShadingModels=Unlit}
 nodes [
  Constant2Vector() {g=0.5, r=0.5}
  MaterialFunctionCall!(Center Location, 0-1 UVs, XY scale=Constant2Vector() {g=85, r=125}, OverridePivot, "Normals (Optional)") {materialFunction=&Sprite}
  n1 = VectorParameter() {parameterName=Color2, defaultValue=<rgba 0.0371 0.498 1 1>}
  n2 = VectorParameter() {parameterName=Color1, defaultValue=<rgba 0.1588 1.1163 1.2 1>}
  n3 = TextureCoordinate()
  n4 = ComponentMask(None=VectorParameter() {defaultValue=<rgba 0 0 0 1>}.RGB) {g=true, r=true}
  n5 = ScalarParameter() {group=Flame, parameterName=Speed X, sliderMax=5, sliderMin=-5}
  n6 = Multiply(
   A=ScalarParameter() {defaultValue=0.4, group=Flame, parameterName=Speed Y, sliderMax=5, sliderMin=-5}
   B=LinearInterpolate(Alpha=NamedRerouteUsage()) {constA=0.9, constB=1.1}
  )
  n7 = ScalarParameter() {defaultValue=2, group=Flame, parameterName=Scale Y, sliderMax=5}
  n8 = LinearInterpolate(Alpha=OneMinus(None=NamedRerouteUsage())) {constA=0.9, constB=1.1}
  n9 = MaterialFunctionCall(
   UVs=Panner(Coordinate=Add(A=@n3, B=@n4), Speed=AppendVector(A=@n5, B=@n6)) {speedX=1}
   Texture Scale=AppendVector(A=ScalarParameter() {defaultValue=1, group=Flame, parameterName=Scale X, sliderMax=5}, B=Multiply(A=@n7, B=@n8))
  ) {materialFunction=&ScaleUVsByCenter}
  n10 = NamedRerouteUsage()
  n11 = Constant()
  n12 = TextureSampleParameter2D(UVs=Add(
    A=MaterialFunctionCall(XY=@n9.UVs) {materialFunction=&Swizzle}.YX
    B=AppendVector(A=@n10, B=@n11)
   )) {parameterName=Mask Texture, texture=/Game/Effects/Fort_Effects/Textures/T_Chainmail_WispyMask}
  n13 = Add(A=TextureCoordinate(), B=AppendVector(A=Constant(), B=NamedRerouteUsage())) {constB=0}
  n14 = VectorParameter() {parameterName=WaveParams, defaultValue=<rgba 0 4.3 0 0.224>}
  n15 = Constant()
  n16 = Constant() {r=1.2}
  n17 = Time()
  n18 = TextureCoordinate()
  n19 = Multiply(
   A=LinearInterpolate(Alpha=MaterialFunctionCall!(
     UV=@n13
     U Frequency=@n14.R
     V Frequency=@n14.R
     U Amplitude=@n14.R
     V Amplitude=@n14.R
     U Speed=@n15
     V Speed=@n16
     U Time=@n17
     V Time=@n17
    ) {materialFunction=/Game/Effects/Fort_Effects/MaterialFunctions/MF_WaveUV}.Wave UVs) {constA=-1}
   B=@n18
  )
  n20 = OneMinus(None=ComponentMask(None=@n18) {g=true})
  n21 = LinearInterpolate(Alpha=Power(Base=@n20) {constExponent=4.5})
  n22 = LinearInterpolate(Alpha=@n20) {constA=1, constB=-0.2}
  n23 = Constant() {r=0.6}
  n24 = NamedRerouteUsage()
  n25 = MaterialFunctionCall(
   UVs=Add(A=Multiply(A=@n19, B=@n21), B=@n18)
   Texture Scale=AppendVector(A=LinearInterpolate(A=@n22, B=@n23, Alpha=@n24), B=LinearInterpolate(Alpha=@n24) {constA=1, constB=0.6})
  ) {materialFunction=&ScaleUVsByCenter}
  n26 = MaterialFunctionCall(Float2=TextureCoordinate()) {materialFunction=&BreakOutFloat2Components}
  n27 = Multiply(A=@n26.R) {constB=0.2}
  n28 = Multiply(A=@n26.G) {constB=0.1}
  n29 = Add(A=Panner(Coordinate=AppendVector(A=@n27, B=@n28)) {speedY=0.18}, B=NamedRerouteUsage())
  n30 = TextureCoordinate()
  n31 = TextureSampleParameter2D!(UVs=@n29, "DDX(UVs)"=DDX(Value=@n30), "DDY(UVs)"=DDY(Value=@n30), Apply View MipBias) {
   group=Main Texture
   mipValueMode=TMVM_Derivative
   parameterName=TexNoise
   samplerType=SAMPLERTYPE_Grayscale
   texture=/Game/Effects/Fort_Effects/Textures/Fire/T_WispyFireNoise_Soft_M
  }
  n32 = OneMinus(None=@n26.G)
  n33 = Multiply(A=@n25.UVs, B=Multiply(A=@n31.R, B=LinearInterpolate(Alpha=@n32) {constA=0.2, constB=2}))
  n34 = NamedRerouteUsage()
  n35 = SphereMask(
   A=Add(A=@n25.UVs, B=Multiply(A=@n33) {constB=0.3})
   B=AppendVector(A=Constant() {r=0.5}, B=LinearInterpolate(Alpha=@n34) {constA=0.5, constB=0.5})
  ) {attenuationRadius=0.4, hardnessPercent=60}
  n36 = Power(Base=@n35) {constExponent=4}
  n37 = TextureCoordinate()
  n38 = Power(
   Base=Multiply(A=@n35, B=LinearInterpolate(A=@n31.R, Alpha=@n36))
   Exp=LinearInterpolate(Alpha=ComponentMask(None=@n37) {g=true}) {constA=3, constB=0.8}
  )
  n39 = Power(Base=ComponentMask(None=TextureCoordinate() {uTiling=4}) {g=true})
  n40 = MaterialFunctionCall(
   Fade Length=ScalarParameter() {defaultValue=20, parameterName=Camera Fade Length, sliderMax=200}
   Fade Offset=ScalarParameter() {defaultValue=20, parameterName=Camera Fade Offset, sliderMax=200}
  ) {materialFunction=&CameraDepthFade}
  n41 = Multiply(A=LinearInterpolate(A=Multiply(A=@n38, B=@n12.R) {constA=1}, B=@n38, Alpha=Saturate(None=@n39)), B=@n40.Result)
  n42 = NamedRerouteUsage()
  n43 = SphereMask(A=@n25.UVs, B=AppendVector(A=Constant() {r=0.5}, B=LinearInterpolate(Alpha=@n42) {constA=0.75, constB=0.5})) {attenuationRadius=0.4, hardnessPercent=0}
  n44 = Saturate(None=Multiply(A=Power(Base=@n43) {constExponent=3}) {constB=2})
  MaterialFunctionCall(Emissive=LinearInterpolate(
    A=LinearInterpolate(A=@n1.RGB, B=@n2.RGB, Alpha=@n41)
    B=VectorParameter() {parameterName=Core Color, defaultValue=<rgba 0.2269 3.6013 4 1>}.RGB
    Alpha=@n44
   )) {materialFunction=&MF_FX_EmissiveAdjustments}
  n45 = Power(Base=@n44) {constExponent=6}
  n46 = MaterialFunctionCall!(UV=TextureCoordinate(), Twist Center, Twist Edges) {materialFunction=/Game/Accessories/FORT_Backpacks/Backpack_F_MED_Celestial_Glow/FX/Materials/MF_PolarUV}
  n47 = Constant2Vector() {g=0.4, r=1}
  n48 = TextureSampleParameter2D(UVs=Panner(Coordinate=Multiply(A=@n46.Polar UVs, B=@n47)) {speedY=-0.15}, "DDX(UVs)"=@n46.DDX, "DDY(UVs)"=@n46.DDX) {
   group=Main Texture
   mipValueMode=TMVM_Derivative
   parameterName=TexNoise
   samplerType=SAMPLERTYPE_Grayscale
   texture=&T_WispyFireNoise_Soft_M
  }
  n49 = TextureCoordinate()
  n50 = Constant() {r=0.5}
  n51 = LinearInterpolate(Alpha=NamedRerouteUsage()) {constA=0.8, constB=0.5}
  n52 = Power(Base=SphereMask(A=Add(A=@n49) {constB=0}, B=AppendVector(A=@n50, B=@n51)) {attenuationRadius=0.35, hardnessPercent=0}) {constExponent=2.5}
  n53 = Multiply(A=LinearInterpolate(B=@n48.R, Alpha=OneMinus(None=@n52)) {constA=1}, B=@n52)
  n54 = NamedRerouteUsage()
  Multiply(
   A=Add(A=Multiply(A=@n45), B=LinearInterpolate(A=@n41, B=@n53, Alpha=@n54))
   B=MaterialFunctionCall(Value_Morning=Constant() {r=0.7}, Value_Day=Constant() {r=1}, Value_Evening=Constant() {r=0.7}, Value_Night=Constant() {r=0.5}) {materialFunction=&Scalar_Time_of_Day}.Result
  )
  n55 = Power(Base=ComponentMask(None=TextureCoordinate()) {r=true}) {constExponent=1}
  n56 = Power(Base=DotProduct(
    A=MaterialFunctionCall() {materialFunction=&CameraDirectionVector}.Result
    B=Constant3Vector() {constant=<rgba 0 0 1 0>}
   ))
  Multiply(A=@n40.Result, B=LinearInterpolate(B=Subtract(A=@n55) {constB=0.1}, Alpha=Saturate(None=@n56)) {constA=1}) {constA=1}
  n57 = ObjectPositionWS()
  NamedRerouteDeclaration(None=Modulo!(
    A=MaterialFunctionCall(Vector 3=@n57) {materialFunction=/Engine/Functions/Engine_MaterialFunctions02/Utility/VectorLength}.V3 Length
    B=Constant() {r=1}
   )) {name=FakePerInstanceRandom, nodeColor=<rgba 0.5937 1 0 1>}
  n58 = MaterialFunctionCall() {materialFunction=&CameraDirectionVector}
  n59 = SmoothStep(Value=ConstantBiasScale(None=MaterialFunctionCall(Float3=@n58.Result) {materialFunction=&BreakOutFloat3Components}.B)) {constMax=0.2}
  NamedRerouteDeclaration(None=OneMinus(None=Saturate(None=@n59))) {name=TopViewFactor, nodeColor=<rgba 0.0078 0 1 1>}
  PerInstanceRandom()
  TextureSampleParameter2D!(UVs=@n29, Apply View MipBias) {parameterName=Noise, samplerType=SAMPLERTYPE_LinearColor, texture=/Game/Global/Textures/Default/Tileable_Patterns/T_PerlinNoise_Uncompressed}
 ]
}
